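# Translucent Default Lit material named purple_field, kept in a Season 17 developer test map folder.
# The graph holds only an unnamed scalar parameter and an unnamed vector parameter with a purple default.
# It has no textures, math or animation, and no links from either parameter to material outputs appear.
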Material /Game/Athena/Apollo/Maps/Test/Dev/Dom/S17/Materials/purple_field
graph {
 properties {blendMode=BLEND_Translucent, pixelDepthOffsetMode=PDOM_Legacy, usedShadingModels=Default Lit}
 nodes [
  ScalarParameter() {defaultValue=0.2}
  VectorParameter() {defaultValue=<rgba 0.607 0.1344 1 0.5062>}
 ]
}
